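# Physics asset for the MechPilotSharkVelocity female face accessory, simulating the character's helmet-off hair around the head.
# Capsule bodies form a center chain, mirrored left and right a, b and c chains, and five single-segment fringe strands, alongside one head capsule.
# Each strand's small root body joins the head with free swing and twist, while following segments use tight angular limits, loosest swing toward the a-chain tips.
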PhysicsAsset /Game/Characters/Player/Female/Medium/Bodies/F_MED_MechPilotSharkVelocity/Meshes/Parts/F_MED_MechPilotSharkVelocity_FaceAcc_Physics
bodies {
 dyn_helm_off_hair_01_c [{shapeType=Capsule, length=2, radius=2, shapeName=dyn_helm_off_hair_01_c_capsule, center=<1 0 0>, rotation=<rot 90 0 0>}]
 dyn_helm_off_hair_01_c_root [{shapeType=Capsule, length=0.505, radius=0.505, shapeName=dyn_helm_off_hair_01_c_root_capsule, rotation=<rot 0 -90 0>}]
 dyn_helm_off_hair_02_c [{shapeType=Capsule, length=2, radius=2, shapeName=dyn_helm_off_hair_02_c_capsule, center=<0.7677 0 0.6403>, rotation=<rot 56.241 180 180>}]
 dyn_helm_off_hair_a_01_l [{shapeType=Capsule, length=2.5, radius=1.5, shapeName=dyn_helm_off_hair_a_01_l_capsule, center=<1.5 0 0>, rotation=<rot 90 0 0>}]
 dyn_helm_off_hair_a_01_l_root [{shapeType=Capsule, length=0.505, radius=0.505, shapeName=dyn_helm_off_hair_a_01_l_root_capsule, rotation=<rot 0 -90 0>}]
 dyn_helm_off_hair_a_01_r [{shapeType=Capsule, length=2.5, radius=1.5, shapeName=dyn_helm_off_hair_a_01_l_capsule, center=<-1.5 0 0>, rotation=<rot 90 180 0>}]
 dyn_helm_off_hair_a_01_r_root [{shapeType=Capsule, length=0.505, radius=0.505, shapeName=dyn_helm_off_hair_a_01_l_root_capsule, rotation=<rot 0 -90 180>}]
 dyn_helm_off_hair_a_02_l [{shapeType=Capsule, length=2.7812, radius=1.2188, shapeName=dyn_helm_off_hair_a_02_l_capsule, center=<1.5 0 0.2135>, rotation=<rot 90 0 0>}]
 dyn_helm_off_hair_a_02_r [{
   shapeType=Capsule
   length=2.7812
   radius=1.2188
   shapeName=dyn_helm_off_hair_a_02_l_capsule
   center=<-1.5 0 -0.2135>
   rotation=<rot 90 0 180>
  }]
 dyn_helm_off_hair_a_03_l [{
   shapeType=Capsule
   length=5.9375
   radius=1.8125
   shapeName=dyn_helm_off_hair_a_03_l_capsule
   center=<3.3859 0.296 0>
   rotation=<rot 90 180 178.0655>
  }]
 dyn_helm_off_hair_a_03_r [{
   shapeType=Capsule
   length=5.9375
   radius=1.8125
   shapeName=dyn_helm_off_hair_a_03_l_capsule
   center=<-3.3859 -0.296 0>
   rotation=<rot 90 180 -1.9345>
  }]
 dyn_helm_off_hair_a_04_l [{shapeType=Capsule, length=2, radius=2, shapeName=dyn_helm_off_hair_a_04_l_capsule, center=<5.2126 0 0>, rotation=<rot 90 0 0>}]
 dyn_helm_off_hair_a_04_r [{shapeType=Capsule, length=2, radius=2, shapeName=dyn_helm_off_hair_a_04_l_capsule, center=<-5.2126 0 0>, rotation=<rot 90 0 180>}]
 dyn_helm_off_hair_b_01_l [{shapeType=Capsule, length=2, radius=2, shapeName=dyn_helm_off_hair_b_01_l_capsule, center=<1 0 0>, rotation=<rot 90 0 0>}]
 dyn_helm_off_hair_b_01_l_root [{shapeType=Capsule, length=0.505, radius=0.505, shapeName=dyn_helm_off_hair_b_01_l_root_capsule, rotation=<rot 0 -90 0>}]
 dyn_helm_off_hair_b_01_r [{shapeType=Capsule, length=2, radius=2, shapeName=dyn_helm_off_hair_b_01_l_capsule, center=<-1 0 0>, rotation=<rot 90 180 0>}]
 dyn_helm_off_hair_b_01_r_root [{shapeType=Capsule, length=0.505, radius=0.505, shapeName=dyn_helm_off_hair_b_01_l_root_capsule, rotation=<rot 0 -90 180>}]
 dyn_helm_off_hair_b_02_l [{shapeType=Capsule, length=2, radius=2, shapeName=dyn_helm_off_hair_b_02_l_capsule, center=<1 0 0>, rotation=<rot 90 0 0>}]
 dyn_helm_off_hair_b_02_r [{shapeType=Capsule, length=2, radius=2, shapeName=dyn_helm_off_hair_b_02_l_capsule, center=<-1 0 0>, rotation=<rot 90 180 0>}]
 dyn_helm_off_hair_b_03_l [{shapeType=Capsule, length=2, radius=2, shapeName=dyn_helm_off_hair_b_03_l_capsule, center=<1 0 0>, rotation=<rot 90 0 0>}]
 dyn_helm_off_hair_b_03_r [{shapeType=Capsule, length=2, radius=2, shapeName=dyn_helm_off_hair_b_03_l_capsule, center=<-1 0 0>, rotation=<rot 90 180 0>}]
 dyn_helm_off_hair_c_01_l [{shapeType=Capsule, length=1.5938, radius=2.4062, shapeName=dyn_helm_off_hair_c_01_l_capsule, center=<1.5764 0 0>, rotation=<rot 90 0 0>}]
 dyn_helm_off_hair_c_01_l_root [{shapeType=Capsule, length=0.505, radius=0.505, shapeName=dyn_helm_off_hair_c_01_l_root_capsule, rotation=<rot 0 -90 0>}]
 dyn_helm_off_hair_c_01_r [{shapeType=Capsule, length=1.5938, radius=2.4062, shapeName=dyn_helm_off_hair_c_01_l_capsule, center=<-1.5764 0 0>, rotation=<rot 90 0 180>}]
 dyn_helm_off_hair_c_01_r_root [{shapeType=Capsule, length=0.505, radius=0.505, shapeName=dyn_helm_off_hair_c_01_l_root_capsule, rotation=<rot 0 -90 180>}]
 dyn_helm_off_hair_c_02_l [{shapeType=Capsule, length=1.5938, radius=2.4062, shapeName=dyn_helm_off_hair_c_02_l_capsule, center=<1.5764 0 0>, rotation=<rot 90 0 0>}]
 dyn_helm_off_hair_c_02_r [{shapeType=Capsule, length=1.5938, radius=2.4062, shapeName=dyn_helm_off_hair_c_02_l_capsule, center=<-1.5764 0 0>, rotation=<rot 90 0 180>}]
 dyn_helm_off_hair_fringe_01 [{shapeType=Capsule, length=1, radius=1, shapeName=dyn_helm_off_hair_fringe_01_capsule, center=<1 0 0.1873>, rotation=<rot 90 0 0>}]
 dyn_helm_off_hair_fringe_01_root [{shapeType=Capsule, length=0.505, radius=0.505, shapeName=dyn_helm_off_hair_fringe_01_root_capsule, rotation=<rot 0 -90 0>}]
 dyn_helm_off_hair_fringe_02 [{shapeType=Capsule, length=1, radius=1, shapeName=dyn_helm_off_hair_fringe_02_capsule, center=<1 0 0.0229>, rotation=<rot 90 0 0>}]
 dyn_helm_off_hair_fringe_02_root [{shapeType=Capsule, length=0.505, radius=0.505, shapeName=dyn_helm_off_hair_fringe_02_root_capsule, rotation=<rot 0 -90 0>}]
 dyn_helm_off_hair_fringe_03 [{shapeType=Capsule, length=1, radius=1, shapeName=dyn_helm_off_hair_fringe_03_capsule, center=<1 0 -0.008>, rotation=<rot 90 0 0>}]
 dyn_helm_off_hair_fringe_03_root [{shapeType=Capsule, length=0.505, radius=0.505, shapeName=dyn_helm_off_hair_fringe_03_root_capsule, rotation=<rot 0 -90 0>}]
 dyn_helm_off_hair_fringe_04 [{shapeType=Capsule, length=1, radius=1, shapeName=dyn_helm_off_hair_fringe_04_capsule, center=<-1 0 0.0423>, rotation=<rot 90 0 0>}]
 dyn_helm_off_hair_fringe_04_root [{shapeType=Capsule, length=0.505, radius=0.505, shapeName=dyn_helm_off_hair_fringe_04_root_capsule, rotation=<rot 0 -90 0>}]
 dyn_helm_off_hair_fringe_05 [{shapeType=Capsule, length=1, radius=1, shapeName=dyn_helm_off_hair_fringe_05_capsule, center=<-1 0 -0.2984>, rotation=<rot 90 0 0>}]
 dyn_helm_off_hair_fringe_05_root [{shapeType=Capsule, length=0.505, radius=0.505, shapeName=dyn_helm_off_hair_fringe_05_root_capsule, rotation=<rot 0 -90 0>}]
 head [{shapeType=Capsule, length=0.625, radius=6.9314, center=<4.6245 0.3234 0>, rotation=<rot -88.6292 32.8783 -107.6165>}]
}
constraints [
 {
  bone1Name=dyn_helm_off_hair_a_01_l_root
  bone2Name=head
  swing1LimitDegrees=45
  swing1Motion=Free
  swing2LimitDegrees=45
  swing2Motion=Free
  twistLimitDegrees=45
  twistMotion=Free
 }
 {bone1Name=dyn_helm_off_hair_a_01_l, bone2Name=dyn_helm_off_hair_a_01_l_root, swing1LimitDegrees=5, swing2LimitDegrees=2, twistLimitDegrees=5}
 {bone1Name=dyn_helm_off_hair_a_02_l, bone2Name=dyn_helm_off_hair_a_01_l, swing1LimitDegrees=5, swing2LimitDegrees=1, twistLimitDegrees=5}
 {bone1Name=dyn_helm_off_hair_a_03_l, bone2Name=dyn_helm_off_hair_a_02_l, swing1LimitDegrees=16, swing2LimitDegrees=1, twistLimitDegrees=8}
 {
  bone1Name=dyn_helm_off_hair_b_01_l_root
  bone2Name=head
  swing1LimitDegrees=45
  swing1Motion=Free
  swing2LimitDegrees=45
  swing2Motion=Free
  twistLimitDegrees=45
  twistMotion=Free
 }
 {bone1Name=dyn_helm_off_hair_b_01_l, bone2Name=dyn_helm_off_hair_b_01_l_root, swing1LimitDegrees=4, swing2LimitDegrees=2, twistLimitDegrees=4}
 {bone1Name=dyn_helm_off_hair_b_02_l, bone2Name=dyn_helm_off_hair_b_01_l, swing1LimitDegrees=4, swing2LimitDegrees=2, twistLimitDegrees=8}
 {bone1Name=dyn_helm_off_hair_b_03_l, bone2Name=dyn_helm_off_hair_b_02_l, swing1LimitDegrees=4, swing2LimitDegrees=2, twistLimitDegrees=8}
 {
  bone1Name=dyn_helm_off_hair_c_01_l_root
  bone2Name=head
  swing1LimitDegrees=45
  swing1Motion=Free
  swing2LimitDegrees=45
  swing2Motion=Free
  twistLimitDegrees=45
  twistMotion=Free
 }
 {bone1Name=dyn_helm_off_hair_c_01_l, bone2Name=dyn_helm_off_hair_c_01_l_root, swing1LimitDegrees=5, swing2LimitDegrees=5, twistLimitDegrees=5}
 {
  bone1Name=dyn_helm_off_hair_c_02_l
  bone2Name=dyn_helm_off_hair_c_01_l
  swing1LimitDegrees=7.5
  swing2LimitDegrees=7.5
  twistLimitDegrees=7.5
 }
 {
  bone1Name=dyn_helm_off_hair_a_01_r_root
  bone2Name=head
  swing1LimitDegrees=45
  swing1Motion=Free
  swing2LimitDegrees=45
  swing2Motion=Free
  twistLimitDegrees=45
  twistMotion=Free
 }
 {
  bone1Name=dyn_helm_off_hair_b_01_r_root
  bone2Name=head
  swing1LimitDegrees=45
  swing1Motion=Free
  swing2LimitDegrees=45
  swing2Motion=Free
  twistLimitDegrees=45
  twistMotion=Free
 }
 {
  bone1Name=dyn_helm_off_hair_c_01_r_root
  bone2Name=head
  swing1LimitDegrees=45
  swing1Motion=Free
  swing2LimitDegrees=45
  swing2Motion=Free
  twistLimitDegrees=45
  twistMotion=Free
 }
 {
  bone1Name=dyn_helm_off_hair_01_c_root
  bone2Name=head
  swing1LimitDegrees=45
  swing1Motion=Free
  swing2LimitDegrees=45
  swing2Motion=Free
  twistLimitDegrees=45
  twistMotion=Free
 }
 {bone1Name=dyn_helm_off_hair_01_c, bone2Name=dyn_helm_off_hair_01_c_root, swing1LimitDegrees=5, swing2LimitDegrees=5, twistLimitDegrees=5}
 {bone1Name=dyn_helm_off_hair_02_c, bone2Name=dyn_helm_off_hair_01_c, swing1LimitDegrees=7.5, swing2LimitDegrees=7.5, twistLimitDegrees=7.5}
 {bone1Name=dyn_helm_off_hair_c_01_r, bone2Name=dyn_helm_off_hair_c_01_r_root, swing1LimitDegrees=2, swing2LimitDegrees=2, twistLimitDegrees=2}
 {bone1Name=dyn_helm_off_hair_a_04_l, bone2Name=dyn_helm_off_hair_a_03_l, swing1LimitDegrees=18, swing2LimitDegrees=1, twistLimitDegrees=4}
 {bone1Name=dyn_helm_off_hair_a_04_r, bone2Name=dyn_helm_off_hair_a_03_r, swing1LimitDegrees=18, swing2LimitDegrees=1, twistLimitDegrees=4}
 {bone1Name=dyn_helm_off_hair_a_03_r, bone2Name=dyn_helm_off_hair_a_02_r, swing1LimitDegrees=16, swing2LimitDegrees=1, twistLimitDegrees=8}
 {bone1Name=dyn_helm_off_hair_a_02_r, bone2Name=dyn_helm_off_hair_a_01_r, swing1LimitDegrees=5, swing2LimitDegrees=1, twistLimitDegrees=5}
 {bone1Name=dyn_helm_off_hair_a_01_r, bone2Name=dyn_helm_off_hair_a_01_r_root, swing1LimitDegrees=5, swing2LimitDegrees=2, twistLimitDegrees=5}
 {bone1Name=dyn_helm_off_hair_b_03_r, bone2Name=dyn_helm_off_hair_b_02_r, swing1LimitDegrees=4, swing2LimitDegrees=2, twistLimitDegrees=8}
 {bone1Name=dyn_helm_off_hair_b_02_r, bone2Name=dyn_helm_off_hair_b_01_r, swing1LimitDegrees=4, swing2LimitDegrees=2, twistLimitDegrees=8}
 {bone1Name=dyn_helm_off_hair_b_01_r, bone2Name=dyn_helm_off_hair_b_01_r_root, swing1LimitDegrees=4, swing2LimitDegrees=2, twistLimitDegrees=4}
 {
  bone1Name=dyn_helm_off_hair_fringe_01_root
  bone2Name=head
  swing1LimitDegrees=45
  swing1Motion=Free
  swing2LimitDegrees=45
  swing2Motion=Free
  twistLimitDegrees=45
  twistMotion=Free
 }
 {
  bone1Name=dyn_helm_off_hair_fringe_01
  bone2Name=dyn_helm_off_hair_fringe_01_root
  swing1LimitDegrees=6
  swing2LimitDegrees=2
  twistLimitDegrees=6
 }
 {
  bone1Name=dyn_helm_off_hair_fringe_02_root
  bone2Name=head
  swing1LimitDegrees=45
  swing1Motion=Free
  swing2LimitDegrees=45
  swing2Motion=Free
  twistLimitDegrees=45
  twistMotion=Free
 }
 {
  bone1Name=dyn_helm_off_hair_fringe_02
  bone2Name=dyn_helm_off_hair_fringe_02_root
  swing1LimitDegrees=6
  swing2LimitDegrees=2
  twistLimitDegrees=6
 }
 {
  bone1Name=dyn_helm_off_hair_fringe_03_root
  bone2Name=head
  swing1LimitDegrees=45
  swing1Motion=Free
  swing2LimitDegrees=45
  swing2Motion=Free
  twistLimitDegrees=45
  twistMotion=Free
 }
 {
  bone1Name=dyn_helm_off_hair_fringe_03
  bone2Name=dyn_helm_off_hair_fringe_03_root
  swing1LimitDegrees=6
  swing2LimitDegrees=2
  twistLimitDegrees=6
 }
 {
  bone1Name=dyn_helm_off_hair_fringe_04_root
  bone2Name=head
  swing1LimitDegrees=45
  swing1Motion=Free
  swing2LimitDegrees=45
  swing2Motion=Free
  twistLimitDegrees=45
  twistMotion=Free
 }
 {
  bone1Name=dyn_helm_off_hair_fringe_04
  bone2Name=dyn_helm_off_hair_fringe_04_root
  swing1LimitDegrees=6
  swing2LimitDegrees=2
  twistLimitDegrees=6
 }
 {
  bone1Name=dyn_helm_off_hair_fringe_05_root
  bone2Name=head
  swing1LimitDegrees=45
  swing1Motion=Free
  swing2LimitDegrees=45
  swing2Motion=Free
  twistLimitDegrees=45
  twistMotion=Free
 }
 {
  bone1Name=dyn_helm_off_hair_fringe_05
  bone2Name=dyn_helm_off_hair_fringe_05_root
  swing1LimitDegrees=6
  swing2LimitDegrees=2
  twistLimitDegrees=6
 }
 {
  bone1Name=dyn_helm_off_hair_c_02_r
  bone2Name=dyn_helm_off_hair_c_01_r
  swing1LimitDegrees=7.5
  swing2LimitDegrees=7.5
  twistLimitDegrees=7.5
 }
]
