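# Animation Blueprint for the MechPilotSharkVelocity face accessory mesh that drives secondary hair motion for helmet-on and helmet-off looks.
# It copies the parent mesh pose, then builds two cached poses: head-relative RigidBody simulation for helmet-off, and chained left/right AnimDynamics bones with angular springs and limits for helmet-on.
# HelmOn picks between them, defaulting to true and set from a Stage1 variant meta tag check when character customization loads.
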AnimBlueprint /BRCosmetics/Characters/Player/Female/Medium/Bodies/F_MED_MechPilotSharkVelocity/Meshes/Parts/F_MED_MechPilotSharkVelocity_FaceAcc_AnimBP
target_skeleton=/Game/Characters/Player/Female/Medium/Bodies/F_MED_MechPilotSharkVelocity/Meshes/Parts/F_MED_MechPilotSharkVelocity_FaceAcc_Skeleton
graphs [
 AnimGraph {nodes [
   Output Pose(Result=Blend Poses by bool(BlendPose_0=Use cached pose 'HelmOnHair'(), BlendPose_1=Use cached pose 'HelmOffHair'(), bActiveValue=Get HelmOn() -> (HelmOn:bool)) {blendType=Linear}) {layerGroup=None}
   Save cached pose 'InPose'(Pose=Copy Pose From Mesh() {bUseAttachedParent=true}) {cacheName=InPose}
   n1 = RigidBody(ComponentPose=Local To Component(LocalPose=Use cached pose 'InPose'())) {
    bOverrideWorldGravity=true
    overridePhysicsAsset=/Game/Characters/Player/Female/Medium/Bodies/F_MED_MechPilotSharkVelocity/Meshes/Parts/F_MED_MechPilotSharkVelocity_FaceAcc_Physics
    simulationSpace=BaseBoneSpace
    alphaBoolBlend {blendOption=Linear}
    baseBoneRef {boneName=head}
    componentAppliedLinearAccClamp=<2400 2400 2400>
    componentLinearAccScale=<0.4 0.4 0.4>
    componentLinearVelScale=<0.4 0.4 0.4>
    overrideWorldGravity {z=-980}
    simSpaceSettings {maxAngularAcceleration=100, maxAngularVelocity=100, maxLinearAcceleration=100, maxLinearVelocity=100, worldAlpha=1}
   }
   Save cached pose 'HelmOffHair'(Pose=Component To Local(ComponentPose=@n1)) {cacheName=HelmOffHair}
   n2 = Anim Dynamics - Bone: dyn_helm_on_hair_01_l(ComponentPose=Local To Component(LocalPose=Use cached pose 'InPose'()), Alpha=1, ExternalForce="0,0,0") {
    angularDampingOverride=1
    angularSpringConstant=1500
    bAngularSpring=true
    bGravityOverrideInSimSpace=true
    bOverrideAngularDamping=true
    gravityScale=0.35
    simulationSpace=RootRelative
    alphaBoolBlend {blendOption=Linear}
    boundBone {boneName=dyn_helm_on_hair_01_l}
    physicsBodyDefinitions [{
      sphereCollisionRadius=0
      boundBone {boneName=dyn_helm_on_hair_01_l}
      boxExtents=<10 20 20>
      constraintSetup {angularLimitsMax=<0 0 10>, angularLimitsMin=<0 -5 -20>, angularTarget=<0 0 0>}
      localJointOffset=<10 0 0>
     }]
   }
   n3 = Anim Dynamics - Bone: dyn_helm_on_hair_01_r(ComponentPose=@n2, Alpha=1, ExternalForce="0,0,0") {
    angularDampingOverride=1
    angularSpringConstant=1500
    bAngularSpring=true
    bGravityOverrideInSimSpace=true
    bOverrideAngularDamping=true
    gravityScale=0.35
    simulationSpace=RootRelative
    alphaBoolBlend {blendOption=Linear}
    boundBone {boneName=dyn_helm_on_hair_01_r}
    physicsBodyDefinitions [{
      sphereCollisionRadius=0
      boundBone {boneName=dyn_helm_on_hair_01_r}
      boxExtents=<10 20 20>
      constraintSetup {angularLimitsMax=<0 0 10>, angularLimitsMin=<0 -5 -20>, angularTarget=<0 0 0>}
      localJointOffset=<-10 0 0>
     }]
   }
   Save cached pose 'HelmOnHair'(Pose=Component To Local(ComponentPose=@n3)) {cacheName=HelmOnHair}
  ]}
 EventGraph {nodes [
   Event OnCharacterCustomizationLoaded()
   > Set HelmOn(HelmOn:bool=CharacterItemDefinitionHasAnyMetaTag(Tags=Make Array("[0]"=(TagName="Cosmetics.Variant.Property.Stage1")))) -> (Output_Get:bool)
   Event BlueprintUpdateAnimation()
  ]}
]
variables [HelmOn {type=boolean, default=true}]
